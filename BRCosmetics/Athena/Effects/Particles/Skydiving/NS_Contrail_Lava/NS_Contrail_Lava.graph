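# Niagara skydiving contrail cosmetic that trails lava-themed ribbons, sprites, meshes and orange lights from the character's skeleton.
# Two four-ribbon emitters split by context: Trail spawns in gameplay with owner-speed velocity scaling, Frontend_Trail001 in the front end, driven by User.FEVec.
# Four lights follow bones each update, while distance-culled sprites and meshes spawn at bones.
# Lava_Ribbons is disabled, as are most Lava_Meshes modules, including initialization and velocity.
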NiagaraSystem /BRCosmetics/Athena/Effects/Particles/Skydiving/NS_Contrail_Lava/NS_Contrail_Lava
dependencies {
 usedDataInterfaces [&NiagaraDataInterfaceCurve, &NiagaraDataInterfaceColorCurve]
 usedDynamicInputs []
 usedModules [&SystemState, /Niagara/Modules/System/SystemLifeCycle]
}
emitters {
 Frontend_Trail001 {
  emitterUpdateScript [
   &EmitterState()
   &FN_SpawnRibbon_v2(
    Number of Ribbons=4
    Spawn Mode=::NewEnumerator0
    SpawnRate:NiagaraFloat=&FN_FrontEndFloatModifier(Gameplay=0, Front End=60)
   )
  ]
  particleSpawnScript [
   &FN_InitializeRibbon_v3(Enable Skeletal Position=true, Lifetime=0.2, Ribbon Width=25)
   &AddVelocity(Velocity=&FN_FrontEndVector3Modifier(Gameplay=<0 0 0>, Front End=&Multiply_VectorByFloat(Vector=$User.FEVec, Float=5.4)))
  ]
  particleUpdateScript [
   &ParticleState()
   &Drag(
    Drag=&FN_FrontEndFloatModifier(
     Gameplay=0
     Front End=&MakeCustomFloatFromBool(
      Bool=&SetBoolByFloatComparison(
       A=&MakeFloatFromVector(Vector=$User.FEVec, Channel=::NewEnumerator1)
       B=-60
       Comparison Type=::NewEnumerator6
      )
      True Float=9
      False Float=0
     )
    )
    Use Linear Drag=true
    Use Rotational Drag=true
   )
   &ScaleVelocity(Velocity Scale=&MakeVector(
     X=0
     Y=0
     Z=&Multiply_Float(
      A=0.9
      B=&FloatFromCurve(
       FloatCurve=DataInterface() {propertyValues {Curve {keys [<at 0 0.9>,<at 1 5>]}}}
       CurveIndex=&Divide_Float(A=&VectorLength(Vector=$Engine.Owner.Velocity), B=9000)
      )
     )
    )) {enabled=false, module=ScaleVelocity001}
   &SolveForcesAndVelocity(Clamp Velocity=false, Limit Acceleration=false)
   &FN_UpdateRibbon(Enable Skeletal Position=true)
   /SetVariables(Particles.Position02:NiagaraPosition=&AddVectorToPosition(
     Position=$Particles.Position
     Vector=&ConvertVectorToPosition(
      Input Position=&FN_FrontEndVector3Modifier(Gameplay=<0 0 -10>, Front End=<-5 0 0>)
      Conversion Operation:ENiagara_LWCConvertVecToPos=::NewEnumerator0
     )
    )) {isSetParameters=true, module=SetVariables_#1}
  ]
  rendererClasses [&NiagaraRibbonRendererProperties]
  renderers [{rendererClass=&NiagaraRibbonRendererProperties}]
 }
 Lava_Meshes {
  emitterSpawnScript [&FN_PreserveSkydiveUserVariables() {enabled=false, module=PreserveSkydiveUserVariables}]
  emitterUpdateScript [
   &EmitterScalability(
    MaxDistance=2000
    MinDistance=500
    Module.bApplySpawnCountScale=false
    Module.bApplySpawnCountScaleByDistanceFraction=false
    Module.bUseMaxDistance=true
    Module.bUseMinDistance=true
    SpawnCountScaleByDistanceFraction=DataInterface() {propertyValues {Curve {keys [<at -0.0224 1>,<at 0.9887 0>]}}}
   ) {enabled=false}
   &EmitterLifeCycle(NextLoopDuration=0) {enabled=false}
   &EmitterState(
    Enable Distance Culling=true
    MaxDistance=2000
    Min Distance Response=::SleepAndLetParticlesFinish
    MinDistance=500
    Module.bUseMaxDistance=true
    Module.bUseMinDistance=true
    Scalability Mode=::NewEnumerator1
   )
   /Game/Effects/Niagara/Modules/Spawn/FN_SkydiveSpawnRate(Global Spawnrate=10) {module=SkydiveSpawnRate}
   &SpawnRate(SpawnRate=20, Use Spawn Probability=false) {enabled=false}
   &SpawnBurst_Instantaneous(Spawn Count=20, Use Spawn Probability=false) {enabled=false}
  ]
  particleSpawnScript [
   /Game/Effects/Niagara/Modules/Groups/FN_SetSpawnGroup_v1/FN_SetSpawnGroup(Override SpawnGroup=true, SpawnGroup:NiagaraInt32=&SpawnGroup())
   /Niagara/Modules/Spawn/Initialization/InitializeParticle(
    Lifetime=&UniformRangedFloat(Minimum=0.4, Maximum=0.7)
    Mesh Scale=&VectorFromFloat(Value=&UniformRangedFloat(Minimum=0.3, Maximum=0.42))
    Module.Write Color=true
    Module.Write Lifetime=true
    Module.Write Mass=false
    Module.Write Material Random=true
    Module.Write Position=false
    Module.Write Scale=true
    Module.Write SpriteRotation=false
    Module.Write SpriteSize=false
    Position=&SimulationPosition()
   ) {enabled=false}
   &AddVelocity(Velocity=&FN_FrontEndVector3Modifier(
     Gameplay=&UniformRangedVector(Minimum=<75 75 240>, Maximum=<-75 -75 400>)
     Front End=&Add_Vector(A=$User.FEVec, B=&UniformRangedVector(Minimum=<50 50 50>, Maximum=<-50 -50 -50>))
    )) {enabled=false}
   &SampleSkeletalMeshSkeleton(
    Mesh:NiagaraDataInterfaceSkeletalMesh=$User.SkeletalMesh
    Specific Bone Index:NiagaraInt32=$Particles.SpawnGroup
    Specific Socket Index:NiagaraInt32=$Particles.SpawnGroup
   )
   /Niagara/Modules/Spawn/Location/SkeletalMeshSkeletonLocation(Module.Inherit Velocity=false, Module.Use Sampled Bone as Ribbon ID=true)
   /SetVariables(Particles.Scale=&Multiply_Vector(
     A=$Particles.Scale
     B=&VectorFromFloat(Value=&OneMinusFloat(Float=$Emitter.Scalability.DistanceFraction))
    )) {enabled=false, isSetParameters=true}
   /Niagara/Modules/Spawn/Orientation/InitialMeshRotation(
    Module.Use Orientation Axis=false
    Pitch:NiagaraFloat=&UniformRangedFloat(Minimum=0, Maximum=360)
    Roll:NiagaraFloat=&UniformRangedFloat(Minimum=0, Maximum=360)
    Yaw:NiagaraFloat=&UniformRangedFloat(Minimum=0, Maximum=360)
   ) {enabled=false}
  ]
  particleUpdateScript [
   /Niagara/Modules/Update/Material/DynamicMatParams(
    Module.Param0WriteEnabled=true
    Module.Param1WriteEnabled=false
    Module.Param2WriteEnabled=false
    Module.Param3WriteEnabled=false
    Param0Value:NiagaraFloat=&FloatFromCurve(FloatCurve=DataInterface() {propertyValues {Curve {keys [<at -0.0046 0>,<at 1 1>]}}})
   ) {enabled=false}
   &Drag(Drag=3, Use Linear Drag=true, Use Rotational Drag=false) {enabled=false}
   &ParticleState()
   &SolveForcesAndVelocity(Clamp Velocity=false, Limit Acceleration=false) {module=SolveForcesAndVelocity001}
   /Niagara/Modules/Update/Orientation/MeshRotationRate(
    Pitch:NiagaraFloat=&UniformRangedFloat(Minimum=0, Maximum=360)
    Roll:NiagaraFloat=&UniformRangedFloat(Minimum=0, Maximum=360)
    Rotation Rate=0.5
    Yaw:NiagaraFloat=&UniformRangedFloat(Minimum=0, Maximum=360)
   ) {enabled=false}
   &SolveForcesAndVelocity(Clamp Velocity=false, Limit Acceleration=false) {enabled=false}
  ]
  rendererClasses [&NiagaraMeshRendererProperties]
  renderers [{rendererClass=&NiagaraMeshRendererProperties, rendererIndex=0}]
 }
 Lava_Ribbons {
  bEnabled=false
  emitterSpawnScript [&FN_PreserveSkydiveUserVariables() {module=PreserveSkydiveUserVariables}]
  emitterUpdateScript [
   &EmitterScalability(
    Module.bApplySpawnCountScale=false
    Module.bApplySpawnCountScaleByDistanceFraction=false
    Module.bUseMaxDistance=false
    Module.bUseMinDistance=false
   ) {enabled=false}
   &EmitterLifeCycle(CompleteOnInactive=true, NextLoopDuration=1) {enabled=false}
   &EmitterState()
   &FN_SkydiveSpawnRate(Global Spawnrate=90) {enabled=false, module=SkydiveSpawnRate}
   &FN_SpawnRibbon_v2(Number of Ribbons=4, Spawn Mode=::NewEnumerator0, SpawnRate=60)
  ]
  particleSpawnScript [
   &FN_SetSpawnGroup(Override SpawnGroup=true, SpawnGroup:NiagaraInt32=&SpawnGroup()) {enabled=false}
   /SetVariables(Particles.ExecutionIndex:NiagaraFloat=&ReturnNormalizedExecIndex()) {isSetParameters=true}
   &InitializeRibbon(
    Lifetime=/Game/Effects/Niagara/DynamicInputs/Skydiving/SkydiveHandsFeetFloatModifier(Hands:NiagaraFloat=0.115, Feet:NiagaraFloat=0.075)
    Module.Write Color=true
    Module.Write Lifetime=true
    Module.Write Mass=true
    Module.Write Material Random=true
    Module.Write Position=false
    Module.Write RibbonTwist=false
    Module.Write RibbonWidth=true
    Position:Vector3f=&SimulationPosition()
    Ribbon Width=&SkydiveHandsFeetFloatModifier(Hands:NiagaraFloat=26, Feet:NiagaraFloat=18)
   ) {enabled=false}
   &FN_InitializeRibbon_v3(
    Enable Skeletal Position=true
    Lifetime:NiagaraFloat=&SkydiveHandsFeetFloatModifier(
     Hands:NiagaraFloat=&Lerp_Float(
      A=0.2
      B=0.05
      Alpha=/BRCosmetics/Effects/Niagara/DynamicInputs/Skydiving/FN_SkydiveDivingLerp(Bone Index:NiagaraInt32=4, Reference Axis:Vector3f=<1 0 0>)
     )
     Feet:NiagaraFloat=0.1
    )
    Ribbon Width:NiagaraFloat=&SkydiveHandsFeetFloatModifier(Hands:NiagaraFloat=26, Feet:NiagaraFloat=18)
   )
   &AddVelocity(Velocity=&FN_FrontEndVector3Modifier(Gameplay=<0 0 0>, Front End=&Multiply_VectorByFloat(Vector=$User.FEVec, Float=5)))
   &SampleSkeletalMeshSkeleton(
    Mesh:NiagaraDataInterfaceSkeletalMesh=$User.SkeletalMesh
    Specific Bone Index:NiagaraInt32=$Particles.SpawnGroup
    Specific Socket Index:NiagaraInt32=$Particles.SpawnGroup
   ) {enabled=false}
   &SkeletalMeshSkeletonLocation(Module.Inherit Velocity=false, Module.Use Sampled Bone as Ribbon ID=true) {enabled=false}
   /Game/Effects/Niagara/Modules/Skydiving/FN_DivingAttributes/FN_DivingAttributes()
  ]
  particleUpdateScript [
   &ParticleState()
   /BRCosmetics/Effects/Niagara/Modules/Ribbons/LockFirstParticleToBone(BoneId:NiagaraInt32=$Particles.SpawnGroup, SkeletalMesh:NiagaraDataInterfaceSkeletalMesh=$User.SkeletalMesh) {enabled=false}
   /Game/Effects/Niagara/Modules/Materials/MaterialPanner() {enabled=false}
   &SolveForcesAndVelocity(Clamp Velocity=false, Limit Acceleration=false)
   &FN_UpdateRibbon(
    Enable Skeletal Position=true
    Initial Color:LinearColor=$Particles.Initial.Color
    Initial RibbonWidth:NiagaraFloat=$Particles.Initial.RibbonWidth
   )
   /SetVariables(Particles.PositionRender:Vector3f=&Add_Vector(
     A=Linked() {linkedVariable {name=Particles.Position, type=NiagaraPosition}}
     B=&FN_FrontEndVector3Modifier(
      Gameplay=<0 0 -10>
      Front End=&Add_Vector(
       A=&Multiply_Vector(A=&TransformVector(Vector=$User.FEVec, Source Space=::World, Destination Space=::Local), B=<0 0 0>)
       B=<-5 0 0>
      )
     )
    )) {isSetParameters=true}
   &DynamicMaterialParameters(
    Index 0 Param 1=$Particles.MaterialRandom
    Module.Param0WriteEnabled=true
    Module.Param1WriteEnabled=true
    Module.Param2WriteEnabled=true
    Module.Param3WriteEnabled=true
   )
  ]
  rendererClasses [&NiagaraRibbonRendererProperties]
  renderers [{rendererClass=&NiagaraRibbonRendererProperties}]
 }
 Lights {
  emitterSpawnScript [&FN_PreserveSkydiveUserVariables() {module=PreserveSkydiveUserVariables}]
  emitterUpdateScript [
   &EmitterLifeCycle(AutoComplete=false, CompleteOnInactive=true, NextLoopDuration=1) {enabled=false}
   &EmitterState()
   &SpawnBurst_Instantaneous(Spawn Count=4, Use Spawn Probability=false)
  ]
  particleSpawnScript [
   &FN_SetSpawnGroup(Emitter Bone Count=4)
   &InitializeParticle(
    Color=<rgba 30 6.5863 0 1>
    Module.Write Color=true
    Module.Write Lifetime=true
    Module.Write Mass=false
    Module.Write Material Random=true
    Module.Write Position=false
    Module.Write Scale=false
    Module.Write SpriteRotation=false
    Module.Write SpriteSize=true
    Position=&SimulationPosition()
   )
   /SetVariables(Particles.LightRadius=30) {isSetParameters=true}
  ]
  particleUpdateScript [
   &SampleSkeletalMeshSkeleton(Mesh:NiagaraDataInterfaceSkeletalMesh=$User.SkeletalMesh, Specific Bone Index:NiagaraInt32=$Particles.SpawnGroup)
   &SkeletalMeshSkeletonLocation(Module.Inherit Velocity=false, Module.Use Sampled Bone as Ribbon ID=true)
   &ParticleState()
  ]
  rendererClasses [&NiagaraLightRendererProperties, &NiagaraSpriteRendererProperties]
  renderers [
   {rendererClass=&NiagaraLightRendererProperties, rendererIndex=0}
   {rendererClass=&NiagaraSpriteRendererProperties, rendererIndex=1}
  ]
 }
 StaticHandSprites {
  emitterSpawnScript [&FN_PreserveSkydiveUserVariables() {module=PreserveSkydiveUserVariables}]
  emitterUpdateScript [
   &EmitterScalability(
    MaxDistance=2000
    Module.bApplySpawnCountScale=false
    Module.bApplySpawnCountScaleByDistanceFraction=false
    Module.bUseMaxDistance=true
    Module.bUseMinDistance=true
    SpawnCountScaleByDistanceFraction=DataInterface() {propertyValues {Curve {keys [<at -0.0098 1>,<at 0.9896 0>]}}}
   ) {enabled=false}
   &EmitterLifeCycle(NextLoopDuration=1) {enabled=false}
   &EmitterState(Enable Distance Culling=true, MaxDistance=2000, Module.bUseMaxDistance=true, Module.bUseMinDistance=false, Scalability Mode=::NewEnumerator1)
   &FN_SkydiveSpawnRate(Global Spawnrate=18) {module=SkydiveSpawnRate}
  ]
  particleSpawnScript [
   &FN_SetSpawnGroup(Override SpawnGroup=true, SpawnGroup:NiagaraInt32=&SpawnGroup())
   &InitializeParticle(
    Lifetime=&UniformRangedFloat(Minimum=0.3, Maximum=0.5)
    Module.Write Color=false
    Module.Write Lifetime=true
    Module.Write Mass=false
    Module.Write Material Random=true
    Module.Write Position=false
    Module.Write Scale=false
    Module.Write SpriteRotation=true
    Module.Write SpriteSize=true
    Position=&SimulationPosition()
    Sprite Rotation=&UniformRangedFloat(Minimum=0, Maximum=360)
    Sprite Size=&Vector2DFromFloat(Value=&UniformRangedFloat(Minimum=25, Maximum=36))
   )
   /SetVariables(Particles.SpriteSize=&Multiply_Vector2(
     A=$Particles.SpriteSize
     B=&Vector2DFromFloat(Value=&OneMinusFloat(Float=$Emitter.Scalability.DistanceFraction))
    )) {isSetParameters=true}
   &AddVelocity(Velocity=&FN_FrontEndVector3Modifier(
     Gameplay=&UniformRangedVector(Minimum=<-10 -10 0>, Maximum=<10 10 200>)
     Front End=&Multiply_VectorByFloat(Vector=$User.FEVec, Float=0.5)
    ))
   &DynamicMaterialParameters(Module.Param0WriteEnabled=true, Module.Param1WriteEnabled=true, Module.Param2WriteEnabled=true, Module.Param3WriteEnabled=true)
   &SampleSkeletalMeshSkeleton(
    Mesh:NiagaraDataInterfaceSkeletalMesh=$User.SkeletalMesh
    Specific Bone Index:NiagaraInt32=$Particles.SpawnGroup
    Specific Socket Index:NiagaraInt32=$Particles.SpawnGroup
   )
   &SkeletalMeshSkeletonLocation(Module.Inherit Velocity=false, Module.Use Sampled Bone as Ribbon ID=true)
  ]
  particleUpdateScript [
   &Color(Color=&ColorFromCurve(ColorCurve=DataInterface() {propertyValues {
       AlphaCurve {keys [<at 0 0 RCIM_Cubic>,<at 1 1 RCIM_Cubic>]}
       BlueCurve {keys [<at -0.0023 1 RCIM_Cubic>,<at 0.7464 0.0871>]}
       GreenCurve {keys [<at -0.0023 1 RCIM_Cubic>,<at 0.7464 0.2561>]}
       RedCurve {keys [<at -0.0023 1 RCIM_Cubic>,<at 0.7464 0.5573>]}
      }}))
   &ParticleState()
   &SolveForcesAndVelocity(Clamp Velocity=false, Limit Acceleration=false)
   &ScaleSpriteSize(Scale Factor=&Vector2DFromFloat(Value=&FloatFromCurve(FloatCurve=DataInterface() {propertyValues {Curve {keys [<at 0 0.7>,<at 1 1.4>]}}})))
  ]
  rendererClasses [&NiagaraSpriteRendererProperties]
  renderers [{rendererClass=&NiagaraSpriteRendererProperties}]
 }
 Trail {
  emitterUpdateScript [
   &EmitterState()
   &FN_SpawnRibbon_v2(
    Number of Ribbons=4
    Spawn Mode=::NewEnumerator0
    SpawnRate:NiagaraFloat=&FN_FrontEndFloatModifier(Gameplay=60, Front End=0)
   )
  ]
  particleSpawnScript [
   &FN_InitializeRibbon_v3(Enable Skeletal Position=true, Lifetime=0.2, Ribbon Width=25)
   &AddVelocity(Velocity=&FN_FrontEndVector3Modifier(Gameplay=<0 0 1000>, Front End=&Multiply_VectorByFloat(Vector=$User.FEVec, Float=5.4)))
  ]
  particleUpdateScript [
   &ParticleState()
   &Drag(
    Drag=&FN_FrontEndFloatModifier(
     Gameplay=0
     Front End=&MakeCustomFloatFromBool(
      Bool=&SetBoolByFloatComparison(
       A=&MakeFloatFromVector(Vector=$User.FEVec, Channel=::NewEnumerator1)
       B=-60
       Comparison Type=::NewEnumerator6
      )
      True Float=10
      False Float=0
     )
    )
    Use Linear Drag=true
    Use Rotational Drag=true
   )
   &ScaleVelocity(Velocity Scale=&MakeVector(
     X=0
     Y=0
     Z=&Multiply_Float(
      A=0.9
      B=&FloatFromCurve(
       FloatCurve=DataInterface() {propertyValues {Curve {keys [<at 0 0.9>,<at 1 5>]}}}
       CurveIndex=&Divide_Float(A=&VectorLength(Vector=$Engine.Owner.Velocity), B=9000)
      )
     )
    )) {module=ScaleVelocity001}
   &SolveForcesAndVelocity(Clamp Velocity=false, Limit Acceleration=false)
   &FN_UpdateRibbon(Enable Skeletal Position=true)
   /SetVariables(Particles.Position02:NiagaraPosition=&AddVectorToPosition(
     Position=$Particles.Position
     Vector=&ConvertVectorToPosition(
      Input Position=&FN_FrontEndVector3Modifier(Gameplay=<0 0 -10>, Front End=<-5 0 0>)
      Conversion Operation:ENiagara_LWCConvertVecToPos=::NewEnumerator0
     )
    )) {isSetParameters=true, module=SetVariables_#1}
  ]
  rendererClasses [&NiagaraRibbonRendererProperties]
  renderers [{rendererClass=&NiagaraRibbonRendererProperties}]
 }
}
summary {
 systemName=NS_Contrail_Lava
 emitters [
  {emitterName=StaticHandSprites, rendererClasses [&NiagaraSpriteRendererProperties]}
  {emitterName=Lava_Meshes, rendererClasses [&NiagaraMeshRendererProperties]}
  {bEnabled=false, emitterName=Lava_Ribbons, rendererClasses [&NiagaraRibbonRendererProperties]}
  {emitterName=Lights, rendererClasses [&NiagaraLightRendererProperties, &NiagaraSpriteRendererProperties]}
  {emitterName=Trail, rendererClasses [&NiagaraRibbonRendererProperties]}
  {emitterName=Frontend_Trail001, rendererClasses [&NiagaraRibbonRendererProperties]}
 ]
 userVariables [
  User.bIsFrontEnd {type=NiagaraBool}
  User.bIsFrontEndPreview {type=NiagaraBool}
  User.FEVec {type=Vector3f, description="", defaultValue {struct=Vector3f, value=<0 0 0>}}
  User.FrontEndVector {type=Vector3f, description="", defaultValue {struct=Vector3f, value=<0 0 1000>}}
  User.PlayerSpeed {type=NiagaraFloat, defaultValue=0, description=""}
  User.RibbonIDOffset {type=NiagaraInt32}
  User.SkeletalMesh {
   type=NiagaraDataInterfaceSkeletalMesh
   defaultValue {
    struct=NiagaraExt_VariableValue_DataInterface
    value {dataInterface=&self:NiagaraDataInterfaceSkeletalMesh_0}
   }
  }
 ]
}
